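; Parent Blueprint for Athena sentry environmental items that detect, track and lose hostile pawns or tagged buildings.
; The target activation volume maintains the in-range list, drives sentry state and starts or stops a targeting ability via gameplay events; lost targets use a delayed return to idle.
; It also handles deactivation, reactivation, EMP tag waits and replicated death.
; Gameplay cues, rotation audio and state registration with the data tracker manager provide feedback.
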(event Custom|Debug_SetIdleBehaviorToOscillates
  ;[0.517 0.1136 1] Debug Events
  ;;[0.5899 1 0.06042] Make sure to call these on the server. Calling on the client will do nothing and we can't do Server MPC events because the sentries are owned by the server.
  (CallFunction|SetIdleBehavior "Oscillates"))

(event Custom|Debug_SetIdleBehaviorToContinuouslyTurnLeft
  ;[0.517 0.1136 1] Debug Events
  ;;[0.5899 1 0.06042] Make sure to call these on the server. Calling on the client will do nothing and we can't do Server MPC events because the sentries are owned by the server.
  (CallFunction|SetIdleBehavior "ContinuouslyTurnLeft"))

(event Custom|Debug_SetIdleBehaviorToContinuouslyTurnRight
  ;[0.517 0.1136 1] Debug Events
  ;;[0.5899 1 0.06042] Make sure to call these on the server. Calling on the client will do nothing and we can't do Server MPC events because the sentries are owned by the server.
  (CallFunction|SetIdleBehavior "ContinuouslyTurnRight"))

(event EventBPOnSentryStateChanged
  (params (PreviousSentryState :type EBuildingGameplayActorSentry_State%20Enum :default "PassiveIdle"))
  (bind (_statetag _bfoundvaliddata) (CallFunction|GetStateTagfromEnumValue))
  ;[0.1 0.1 1] State Change
  (CallFunction|LogMessage (Utilities|String|Append "BP event sentry state changed // Previous State = " (Utilities|String|EnumtoString PreviousSentryState) " // Current State = " (Utilities|String|EnumtoString (Variables|BuildingGameplayActorSentry|GetCurrentSentryState)) " // Actor = " (Utilities|GetDisplayName self)))
  (Utilities|FlowControl|Sequence
    (:then_0
      (Networking|CanExecuteCosmeticEvents
        (:True
          (bind _returnvalue (FortGameplayDataTrackerComponent|GetFortGameplayDataTrackerComponentManager))
          (Utilities|IsValid _returnvalue
            (:Is%20Valid
              (if _bfoundvaliddata
                (Class|FortGameplayDataTrackerComponentManager|SetTrackedActorState _returnvalue self _statetag)))))))
    (:then_1
      ;;[0 0 0] On State Change if it is deactivate then call the deactivate function
      (if (Utilities|Enum|Equal(Enum) (Variables|BuildingGameplayActorSentry|GetCurrentSentryState) "Deactivated")
        (Variables|Default|SetHasBeenDeactivated true)
        (CallFunction|Deactivate)
        (else
          (Variables|Default|SetHasBeenDeactivated false))))))

(event Custom|BeginTargeting
  ;[0.1 0.1 1] Begin Targeting - Set the state to ActiveIdle and activate the targeting GA
  ; End Targeting - Let the GA know to stop targeting
  ;;[0 0 0]
  (CallFunction|LogMessage "Sentry Begin Targeting")
  (Ability|SendGameplayEventtoActor self "(TagName=\"Event.Athena.EnvItem.Sentry.Target.Begin\")" (Utilities|Struct|MakeGameplayEventData :ContextHandle "()" :InstigatorTags "(GameplayTags=)" :TargetTags (Variables|SentryParentVars|GetTC_IgnoreTargetingTags) :EventMagnitude 0.0 :TargetData "()")))

(event Custom|EndTargeting
  ;[0.1 0.1 1] Begin Targeting - Set the state to ActiveIdle and activate the targeting GA
  ; End Targeting - Let the GA know to stop targeting
  ;;[0 0 0]
  (CallFunction|LogMessage "Sentry End Targeting")
  (Ability|SendGameplayEventtoActor self "(TagName=\"Event.Athena.EnvItem.Sentry.Target.End\")"))

(event Custom|LostTarget
  ;[0.1 0.1 1] Lost Target - Clear out the targeted pawn, start timer to delay before returning to idle.
  ;;[0 0 0] When we lose our target
  (if (Networking|IsServer)
    (Utilities|Time|ClearandInvalidateTimerbyHandle (Variables|SentryParentVars|GetTimer_LostTarget))
    (CallFunction|SetTargetLastLoc)
    (CallFunction|SetActortoTrack 0)
    (Ability|ExecuteGameplayCueReplicated (Variables|ParentGCs|GetGC_LostTarget))
    (bind _returnvalue (Utilities|Time|SetTimerbyEvent (EventDispatchers|CreateEvent :!event Custom|PostLostTargetDelay) (ScalableFloat|GetValueatLevel (Variables|ParentRows|GetRow_LostTargetDelay))))
    (Variables|SentryParentVars|SetTimer_LostTarget _returnvalue)))

(event OnLinkedActorDestroyed(FortLinkToActor)
  (params (DamageCauser :type Actor%20Object%20Reference))
  ;[0 0 0 1 25] Damage and Death
  ;;[1 0.25 0] On Linked Actor Destroyed
  ;; Rep Death
  (Variables|SentryParentVars|SetRepDeath true))

(event EventOnDamageServer
  (params (Damage :type Float%20%28single-precision%29) (DamageTags :type Gameplay%20Tag%20Container%20Structure%20%28by%20ref%29) (Momentum :type Vector :default "0, 0, 0") (HitInfo :type Hit%20Result%20Structure%20%28by%20ref%29) (InstigatedBy :type Controller%20Object%20Reference) (DamageCauser :type Actor%20Object%20Reference) (EffectContext :type Gameplay%20Effect%20Context%20Handle%20Structure))
  (bind (_bblockinghit _binitialoverlap _time _distance _location _impactpoint _normal) (Collision|BreakHitResult HitInfo))
  ;[0 0 0 1 25] Damage and Death
  ;;[1 0 0.75] On Damaged Server - Play Damaged Cue. Hide behind a do once with a
  ;; delay so that it only can trigger every Nth of a second.
  (Utilities|FlowControl|DoOnce false
    (:Completed
      (Utilities|FlowControl|Sequence
        (:then_0
          _)
        (:then_1
          (Utilities|FlowControl|Delay 0.2)))
      (Ability|ExecuteGameplayCueWithParamsReplicated (Variables|ParentGCs|GetGC_Damaged) (Ability|GameplayCue|MakeGameplayCueParameters :NormalizedMagnitude 0.0 :RawMagnitude 0.0 :Location _location :Normal _normal :GameplayEffectLevel 1 :AbilityLevel 1 :bReplicateLocationWhenUsingMinimalRepProxy false)))))

(event EventOnDeathServer
  (params (Damage :type Float%20%28single-precision%29) (DamageTags :type Gameplay%20Tag%20Container%20Structure%20%28by%20ref%29) (Momentum :type Vector :default "0, 0, 0") (HitInfo :type Hit%20Result%20Structure%20%28by%20ref%29) (InstigatedBy :type Controller%20Object%20Reference) (DamageCauser :type Actor%20Object%20Reference) (EffectContext :type Gameplay%20Effect%20Context%20Handle%20Structure))
  ;[0 0 0 1 25] Damage and Death
  ;;[1 0.25 0] On Actor Death Rep Death
  (Variables|SentryParentVars|SetRepDeath true))

(event Custom|PostLostTargetDelay
  ;[0.1 0.1 1] Lost Target - Clear out the targeted pawn, start timer to delay before returning to idle.
  ;;[0 0 0] After effects of losing target
  (CallFunction|PlayLostTargetSound)
  (Variables|BuildingGameplayActorSentry|GetActorBeingTracked
    (:then)
    (:else
      (CallFunction|SetSentryState "ReturningToIdle")
      (if (Variables|BuildingGameplayActorSentry|GetbDoesNotRotate)
        (Utilities|FlowControl|Delay (Variables|SentryParentVars|GetDelay_NoRotateReset))
        (CallFunction|SetSentryState (select (<= (Utilities|Array|Length (Variables|BuildingGameplayActorSentry|GetTargetActorsInRange)) 0) (select (Variables|BuildingGameplayActorSentry|GetbGoDormantAfterPassiveIdle) "Dormant" "PassiveIdle") "ActiveIdle"))))))

(event Custom|HideAndKill
  ;[0 0 0 1 25] Damage and Death
  ;;[1 0.25 0] Hide And Kill - Play Death GC,
  ;; Disable tick, Hide Actor, Turn off collision
  (Ability|ExecuteGameplayCueReplicated (Variables|ParentGCs|GetGC_Death))
  (Actor|Tick|SetActorTickEnabled false)
  (Rendering|SetActorHiddenInGame true)
  (Collision|SetActorEnableCollision false)
  (CallFunction|EndTargeting)
  (CallFunction|SetActortoTrack 0)
  (Actor|DestroyActor)
  ;;[1 0 0.75] Stop audio, If playing when destroyed
  (bind _rotation_audio (Variables|Default|GetRotation%20Audio)
    (:then
      (if (Audio|Components|Audio|IsPlaying _rotation_audio)
        (Audio|Components|Audio|FadeOut _rotation_audio 0.25 0.0 "Linear" :!class /Script/Engine.AudioComponent)))
    (:else)))

(event Custom|Play%20Lost%20Target%20Sound
  ;[0.1 0.1 1] Lost Target - Clear out the targeted pawn, start timer to delay before returning to idle.
  ;;[1 0 0.75] Lost Target Sound
  (Audio|PlaySoundatLocation (Variables|Sound|GetLost%20Target%20Sound) (Transformation|GetActorLocation)))

(event EventEndPlay
  (params (EndPlayReason :type EEndPlayReason%20Enum :default "Destroyed"))
  ;[0 0 0 1 25] Damage and Death
  ;;[1 0.25 0] EndPlay
  (|CallParentFunction EndPlayReason)
  (Networking|CanExecuteCosmeticEvents
    (:True
      (bind _returnvalue (FortGameplayDataTrackerComponent|GetFortGameplayDataTrackerComponentManager))
      (Utilities|IsValid _returnvalue
        (:Is%20Valid
          (Class|FortGameplayDataTrackerComponentManager|UnRegisterTrackedStateActor _returnvalue self))))))

(event Custom|ValidTargetNearby
  (params (OverlappingActor :type Actor%20Object%20Reference))
  ;[0 0 0] Volume Overlaps
  ;;[0.2 1 0.2] Begin Overlap (Pawn Volume) - Check if it is a player pawn, if so add it to the tracking array set to active idle and begin targeting
  (bind _asgameplay_tag_asset_interface (Utilities|Casting|CastToGameplayTagAssetInterface OverlappingActor))
  (Utilities|TagCheckCanActivateVolume _asgameplay_tag_asset_interface
    (:Pass
      (Utilities|Array|AddUnique (Variables|BuildingGameplayActorSentry|GetTargetActorsInRange) OverlappingActor)
      (Utilities|FlowControl|Sequence
        (:then_0
          (switch Utilities|FlowControl|Switch|SwitchonEBuildingGameplayActorSentry_State (Variables|BuildingGameplayActorSentry|GetCurrentSentryState)
            (:PassiveIdle
              _)
            (:Dormant
              _))
          (CallFunction|SetSentryState "ActiveIdle"))
        (:then_1
          (CallFunction|BeginTargeting) ; Starts GA for targeting
        )))))

(event OnComponentBeginOverlap(TargetActivationVolume)
  (params (OverlappedComponent :type Primitive%20Component%20Object%20Reference) (OtherActor :type Actor%20Object%20Reference) (OtherComp :type Primitive%20Component%20Object%20Reference) (OtherBodyIndex :type Integer) (bFromSweep :type Boolean) (SweepResult :type Hit%20Result%20Structure%20%28by%20ref%29))
  ;[0 0 0] Volume Overlaps
  ;;[0.2 1 0.2] Begin Overlap (Pawn Volume) - Check if it is a player pawn, if so add it to the tracking array set to active idle and begin targeting
  (bind _isvalidtargetactor (CallFunction|IsActorTargetable OtherActor))
  (if _isvalidtargetactor
    (CallFunction|ValidTargetNearby OtherActor)))

(event Custom|Deactivate
  ;[0.1 0.1 1] Deactivate - Called when deactivated to turn off the sentry
  (Ability|ExecuteGameplayCueReplicated (Variables|ParentGCs|GetGC_Deactivated))
  (Actor|Tick|SetActorTickEnabled false)
  (Collision|SetCollisionEnabled (Variables|BuildingGameplayActorSentry|GetTargetActivationVolume) "NoCollision")
  (Utilities|Time|ClearandInvalidateTimerbyHandle (Variables|SentryParentVars|GetTimer_LostTarget))
  (CallFunction|EndTargeting)
  (CallFunction|SetActortoTrack 0)
  (Utilities|Array|Clear (Variables|BuildingGameplayActorSentry|GetTargetActorsInRange)))

(event Custom|EnableTargetActivationVolume
  ;[0 0 0] Volume Overlaps
  ;;[0.2 1 0.2] Begin Overlap (Pawn Volume) - Check if it is a player pawn, if so add it to the tracking array set to active idle and begin targeting
  ;;;[1 1 0] Turning the volume back on
  (Collision|SetCollisionEnabled (Variables|BuildingGameplayActorSentry|GetTargetActivationVolume) "QueryOnly"))

(event EventTryReactivate
  ;[0.1 0.1 1] Try Reactivate
  (if (CallFunction|CanReactivate)
    (Actor|Tick|SetActorTickEnabled true)
    (Networking|SwitchHasAuthority
      (:Authority
        (CallFunction|SetSentryState (CallFunction|GetStatetoSetonReactivate))
        (CallFunction|EnableTargetActivationVolume)))))

(event EventTick
  (params (DeltaSeconds :type Float%20%28single-precision%29))
  (|CallParentFunction DeltaSeconds))

(event OnComponentEndOverlap(TargetActivationVolume)
  (params (OverlappedComponent :type Primitive%20Component%20Object%20Reference) (OtherActor :type Actor%20Object%20Reference) (OtherComp :type Primitive%20Component%20Object%20Reference) (OtherBodyIndex :type Integer))
  ;[0 0 0] Volume Overlaps
  ;;[1 0.2 0.2] End Overlap (Pawn Volume) - Check if what ended the overlap was an actor in our tracking array and remove it.
  ;; Check the size of the array to see if we have anyone in the volume, if not and the target that was removed was the tracking pawn then send the lost target event.
  ;; If it was the last target stop doing the targeting and set the state to be the passive state. Also if we have a tracking pawn call Lost Target
  (bind (_remainingactorsinrange _bdidremoveactor) (CallFunction|RemoveTargetActor OtherActor))
  (if (and _bdidremoveactor (Utilities|Enum|NotEqual(Enum) (Variables|BuildingGameplayActorSentry|GetCurrentSentryState) "Deactivated"))
    (if (Utilities|Array|IsEmpty (Variables|BuildingGameplayActorSentry|GetTargetActorsInRange))
      (CallFunction|EndTargeting) ; This only stops targeting GA
      (Variables|BuildingGameplayActorSentry|GetActorBeingTracked
        (:then
          (CallFunction|LostTarget))
        (:else
          (CallFunction|SetSentryState "ReturningToIdle")))
      (else
        ;;;[0 0 0] Call lost target if the end overlap actor is our current tracked actor
        (if (Utilities|Equal(Object) OtherActor (Variables|BuildingGameplayActorSentry|GetActorBeingTracked))
          (CallFunction|LostTarget))))))

(event EventSetNewOwner
  (params (NewOwner :type Actor%20Object%20Reference))
  ;[0 0 0] Owner
  ;;[1 1 0] Set New Owner
  (Actor|SetOwner NewOwner)
  (bind _output_get (Variables|SentryParentVars|SetSentryOwner NewOwner))
  (Utilities|IsValid _output_get
    (:Is%20Valid
      (CallFunction|SetTeam (Team|GetActorTeam _output_get)))))

(event EventPlayStartedIdleRotationAudio
  ;[0 0 0] Cosmetics
  ;;[1 0 0.75] Rotation Audio
  (Audio|PlaySoundatLocation (Variables|Sound|GetRotation%20Sound%20Start) (Transformation|GetActorLocation))
  (Audio|Components|Audio|FadeIn (Variables|Default|GetRotation%20Audio) 0.3 1.0 0.0 "Linear" :!class /Script/Engine.AudioComponent))

(event Custom|OnRep_SentryOwner)

(event EventPlayStoppedIdleRotationAudio
  ;[0 0 0] Cosmetics
  ;;[1 0 0.75] Rotation Audio
  (Audio|PlaySoundatLocation (Variables|Sound|GetRotation%20Stop%20Sound) (Transformation|GetActorLocation))
  (Audio|Components|Audio|FadeOut (Variables|Default|GetRotation%20Audio) 0.2 0.0 "Linear" :!class /Script/Engine.AudioComponent))

(event Custom|BP%20Stop%20Rotation%20Sound
  ;[0 0 0] Cosmetics
  ;;[1 0 0.75] Rotation Audio
  (Audio|PlaySoundatLocation (Variables|Sound|GetRotation%20Stop%20Sound) (Transformation|GetActorLocation))
  (Audio|Components|Audio|FadeOut (Variables|Default|GetRotation%20Audio) 0.2 0.0 "Linear" :!class /Script/Engine.AudioComponent))

(event Custom|DeactivateRays
  ;[0 0 0] Cosmetics
  ;;[1 0 0.75] Deactivate Rays - Scale Down the Rays when deactivated
  (if (not (Networking|IsServer))
    (Development|SetHiddeninGame (Variables|BuildingGameplayActorSentry|GetAngleRaysStaticMeshComp) true false)))

(event Custom|ActivateRays
  ;[0 0 0] Cosmetics
  ;;[1 0 0.75] Deactivate Rays - Scale Down the Rays when deactivated
  (if (not (Networking|IsServer))
    (Development|SetHiddeninGame (Variables|BuildingGameplayActorSentry|GetAngleRaysStaticMeshComp) false false)))

(event EventAttemptToRegisterWithDataTrackerManager
  (bind (_statetag _bfoundvaliddata) (CallFunction|GetStateTagfromEnumValue))
  ;[1 1 1] Event to register to data tracker manager
  (bind _returnvalue (FortGameplayDataTrackerComponent|GetFortGameplayDataTrackerComponentManager))
  (Utilities|IsValid _returnvalue
    (:Is%20Valid
      (Class|FortGameplayDataTrackerComponentManager|RegisterTrackedStateActor _returnvalue self (select _bfoundvaliddata _statetag (GameplayTags|MakeLiteralGameplayTag "(TagName=\"NPC.AlertLevel.Unaware\")"))))
    (:Is%20Not%20Valid
      (Utilities|Time|SetTimerbyEvent (delegate EventAttemptToRegisterWithDataTrackerManager) 5.0) ; Retry if the manager does not exist yet
    )))

(event EventSetupWaitTasksForEmpTags
  ;[1 1 0] Setup wait tasks for emp tags
  (Ability|Tasks|WaitGameplayTagAddToActor 0 (Variables|Tags|GetTag_EMPApplied)
    (:then
      (Ability|Tasks|WaitGameplayTagRemoveFromActor 0 (Variables|Tags|GetTag_EMPApplied)
        (:Removed
          (CallFunction|SetEmpApplied false))))
    (:Added
      (CallFunction|SetEmpApplied true))))

(fn CanAttackTargetActor
  (params (Actor :type Actor%20Object%20Reference))
  (bind _asfort_pawn (Utilities|Casting|CastToFortPawn Actor)
    (:then
      (return (not (Math|Boolean|ORBoolean (Variables|/Script/FortniteGame.FortPawn|GetbIsDBNO _asfort_pawn) (FortPawn|IsDead _asfort_pawn) (GameplayTags|HasAnyMatchingGameplayTags _asfort_pawn (Variables|Default|GetTag%20Container))))))
    (:CastFailed
      (Utilities|Casting|CastToBuildingActor Actor
        (:then
          (return true))
        (:CastFailed
          (return false))))))

(fn CanDoTargeting
  (bind _returnvalue (|CallParentFunction))
  (return _returnvalue))

(fn CanReactivate
  (return (and (Variables|Default|GetHasBeenDeactivated) (not (GameplayTags|HasAnyMatchingGameplayTags (Variables|SentryParentVars|GetTC_DeactivatedReasons))))))

(fn GetStateTagFromEnumValue
  (bind (_value _returnvalue) (Utilities|Map|Find (Variables|Default|GetStateTagMap) (Variables|BuildingGameplayActorSentry|GetCurrentSentryState)))
  (return _value _returnvalue))

(fn GetStateToSetOnReactivate
  (return "ReturningToIdle"))

(fn GetTargetOwner
  (return 0))

(fn IsActorTargetable
  (params (ActorCandidate :type Actor%20Object%20Reference))
  (Utilities|FlowControl|Sequence
    (:then_0
      ;[0.3438 0.2756 0.0905] NOT Valid if the input actor is nullptr
      (Utilities|IsValid ActorCandidate
        (:Is%20Not%20Valid
          (return false))))
    (:then_1
      ;[0.4375 0.09302 0] We cannot target ourself.
      (if (== self ActorCandidate)
        (return false)))
    (:then_2
      ;[0.3125 0.1806 0.07082] NOT Valid if A) our teams are not hostile towards each other AND we are either on the same team or this actor is a spectator.
      (if (and (Utilities|Enum|NotEqual(Enum) (Team|GetTeamAffiliationBetweenTeams (Team|GetActorTeam ActorCandidate) (Team|GetActorTeam 0)) "Hostile") (or (== (Team|GetActorTeam ActorCandidate) (Team|GetActorTeam 0)) (Utilities|Enum|NotEqual(Enum) (Utilities|Enum|BytetoEnumEFortTeam (Team|GetActorTeam 0)) (Utilities|Enum|LiteralenumEFortTeam "Spectator"))))
        (return false)))
    (:then_3
      ;[0.02893 0 1] Player Pawns that passed the previous checks are valid.
      (Utilities|Casting|CastToFortPlayerPawn ActorCandidate)
      (return true))
    (:then_4
      ;[0.3408 0.3565 1] If we can target buildings, and this is a building with the proper tag, then it is a valid actor.
      (if (Variables|Default|GetCanTargetBuildings%20Const)
        (Utilities|Casting|CastToBuildingGameplayActor ActorCandidate)
        (bind _returnvalue (GameplayTags|HasMatchingGameplayTag(Message) ActorCandidate "(TagName=\"Abilities.Generic.SentryGun.BuildingCanBeTargeted\")"))
        (return _returnvalue)))
    (:then_5
      ;[1 0.4562 0.4479] When all else fails, this actor is NOT valid, return false.
      (return false))))

(fn IsUpsideDown
  (return (< (Math|Vector|DotProduct "0, 0,1.000000" (Transformation|GetActorUpVector)) -0.5)))

(fn IsValidAutoFireTarget
  (params (Attacker :type Actor%20Object%20Reference))
  (return (Utilities|Select :Index (Variables|BuildingGameplayActorSentry|GetCurrentSentryState) :PassiveIdle true :ActiveIdle true :Tracking true :Aggro true :Dormant false :Deactivated false :ReturningToIdle true :LocatingDamager false)))

(fn LogMessage
  (params (InString :type String))
  (if (Variables|Default|GetLoggingEnabled)
    (Development|PrintString (Utilities|String|Append "BGA_Athena_Sentry_Parent [" InString "].") true true "(R=0.000000,G=0.660000,B=1.000000,A=1.000000)" 5.0)))

(fn LookLocationToWorldRotation
  (params (Location :type Vector))
  (bind (_returnvalue_roll _returnvalue_pitch) (Math|Rotator|BreakRotator (Transformation|GetWorldRotation (Variables|BuildingGameplayActorSentry|GetCameraStaticMeshComp))))
  (bind (_returnvalue_roll_2 _returnvalue_pitch_2 _returnvalue_yaw) (Math|Rotator|BreakRotator (Math|Rotator|MakeRotfromX (Math|Vector|Normalize (- (Math|Transform|TransformLocation (Transformation|GetActorTransform) Location) (Transformation|GetActorLocation))))))
  (return _returnvalue_roll_2 _returnvalue_pitch _returnvalue_yaw))

(fn OnRep_RepDeath
  (if (Variables|SentryParentVars|GetRepDeath)
    (CallFunction|HideandKill)
    (CallFunction|BPStopRotationSound)))

(fn RemoveTargetActor
  (params (ActorToRemove :type Actor%20Object%20Reference%20%28by%20ref%29))
  (bind _targetactorsinrange (Variables|BuildingGameplayActorSentry|GetTargetActorsInRange))
  ;[0 0 0] Remove Actor from our target list
  (bind _actortoremove (Variables|GetActorToRemove)
    (:then
      (bind _returnvalue (Utilities|Array|RemoveItem _targetactorsinrange _actortoremove))
      (return _targetactorsinrange _returnvalue))
    (:else
      (return _targetactorsinrange false))))

(fn ScalePosYawCamera
  (params (YawIn :type Float%20%28double-precision%29))
  (return (Math|Float|SelectFloat (+ YawIn 3.6e+02) YawIn (< YawIn 0.0))))

(fn SetTargetLastLoc
  (bind _actorbeingtracked (Variables|BuildingGameplayActorSentry|GetActorBeingTracked))
  (Utilities|IsValid _actorbeingtracked
    (:Is%20Valid
      (Variables|SentryParentVars|SetLastKnownTargetLocation (Transformation|GetActorLocation _actorbeingtracked)))))

(fn ShouldAlliesAutoFireOnDeployedTarget
  (return false))

(fn ShouldAutoFireOnSelfDeployedTarget
  (return false))

(fn ConstructionScript
  (bind (_rightlookidle_x _rightlookidle_y) (Math|Vector|BreakVector (Variables|GetRightLookIdle_X)))
  (bind (_leftlookidle_x _leftlookidle_y) (Math|Vector|BreakVector (Variables|GetLeftLookIdle_X)))
  (|CallParentFunction)
  ;[1 1 1] Turn off collision for the pawn activation volume on the client
  (if (not (Networking|IsServer))
    (Variables|/Script/Engine.PrimitiveComponent|SetbGenerateOverlapEvents false (Variables|BuildingGameplayActorSentry|GetTargetActivationVolume)))
  ;[1 1 1] Set the rotation rates based on the exposed vector actors
  (Variables|SetLeftLookIdle_X (Math|Vector|MakeVector (Math|Float|Clamp(Float) _leftlookidle_x -512.0 512.0) (Math|Float|Clamp(Float) _leftlookidle_y -512.0 512.0)))
  (Variables|SetRightLookIdle_X (Math|Vector|MakeVector (Math|Float|Clamp(Float) _rightlookidle_x -512.0 512.0) (Math|Float|Clamp(Float) _rightlookidle_y -512.0 512.0)))
  ;[0 0 0] Set this on both the client and server so the client has an initial value right away.
  ; (Probably doesn't even need to be replicated, but since it has been replicated since 2020, the class has multiple child bps, and the component's scale is changed both in BP and code I leaving it replicated for safety for now. )
  ;
  ; We have to do this during construction because AngleRayStaticMeshComp's scale can be altered before BeginPlay.
  (Variables|SentryParentVars|SetStartingScaleRays (Variables|/Script/Engine.SceneComponent|GetRelativeScale3D (Variables|BuildingGameplayActorSentry|GetAngleRaysStaticMeshComp))))
