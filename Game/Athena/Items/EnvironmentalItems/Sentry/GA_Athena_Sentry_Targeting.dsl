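; Gameplay ability that drives target acquisition and tracking for Athena sentries until a Target.End event arrives.
; A repeating timer filters in-range actors through a view cone with optional line of sight.
; Without a target, it clears the aggro timer and tracks candidates passing ignore-tag and sentry attack checks.
; A tracked target gets one missed tick before LostTarget fires; overridable functions allow custom filtering, restrictions and found-target behavior.
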(event Ability|EventActivateAbilityFromEvent
  (params (EventData :type Gameplay%20Event%20Data%20Structure%20%28by%20ref%29))
  (bind (_eventtag _instigator _target _optionalobject _optionalobject2 _contexthandle _instigatortags _targettags) (Utilities|Struct|BreakGameplayEventData EventData))
  (bind _returnvalue (Ability|GetAvatarActorfromActorInfo))
  ;[0.2069 1 0.4169] Activate Ability - Start Targeting
  (Utilities|FlowControl|Sequence
    (:then_0
      (Variables|Default|SetSentryTargetingIgnoreTags _targettags)
      ;;[1 1 1] Save Out Ref to Sentry for later use so we never recast
      (Utilities|IsValid (Variables|Default|GetSelfSentryRef)
        (:Is%20Not%20Valid
          (Utilities|IsValid _returnvalue
            (:Is%20Valid
              (bind _asbga_athena_sentry_parent (Utilities|Casting|CastToBGA_Athena_Sentry_Parent _returnvalue))
              (Variables|Default|SetSelfSentryRef _asbga_athena_sentry_parent))))))
    (:then_1
      (bind _returnvalue_2 (Utilities|Time|SetTimerbyEvent (EventDispatchers|CreateEvent :!event Custom|DoTargeting) (Variables|Default|GetTargetingInterval) true false 0.0 (Variables|Default|GetTargetingInitialStartDelayVariance))) ; Start the DoTargeting timer.
      (Variables|Default|SetTargetingTimer _returnvalue_2)
      (Ability|Tasks|WaitGameplayEvent "(TagName=\"Event.Athena.EnvItem.Sentry.Target.End\")"
        (:EventReceived
          (Ability|EndAbility))) ; Wait for end targeting
    )))

(event Ability|EventOnEndAbility
  (params (bWasCancelled :type Boolean))
  ;[1 0.5826 0.3202] End Ability
  (Utilities|Time|ClearandInvalidateTimerbyHandle (Variables|Default|GetTargetingTimer)))

(event Custom|DoTargeting
  (bind _selfsentryref (Variables|Default|GetSelfSentryRef))
  (bind _actorbeingtracked (Variables|/Script/FortniteGame.BuildingGameplayActorSentry|GetActorBeingTracked _selfsentryref))
  (bind _selfsentryref_2 (Variables|Default|GetSelfSentryRef))
  ;[0.1774 0.5943 1] DoTargeting
  (Utilities|IsValid _selfsentryref_2
    (:Is%20Valid
      (bind _returnvalue (Class|BuildingGameplayActorSentry|CanDoTargeting _selfsentryref_2)) ; Double check we can do targeting
      (if _returnvalue
        ;;[1 1 1] Take current targets in range and filter them by the cone.
        (bind _returnvalue_2 (CallFunction|FilterTargetsInsideCone (Variables|Default|GetConeRadius) (Variables|Default|GetConeDistance) (Variables|Default|GetUseLos) "(GameplayTags=((TagName=\"Athena.Vehicle\"),(TagName=\"Gameplay.Melee.Block\"),(TagName=\"Pawn.Athena.NPC.Wildlife\")))"))
        (Utilities|IsValid _actorbeingtracked
          (:Is%20Valid
            ;;[1 1 1] If we have a valid target pawn then check if we still can find the pawn.
            (bind _allowtracking (CallFunction|CheckAttackRestrictions _actorbeingtracked))
            (if _allowtracking
              (if (Utilities|Array|ContainsItem _returnvalue_2 _actorbeingtracked)
                (Variables|Default|SetLostTarget false)
                (CallFunction|HandleFoundTrackedActor) ; Override this for additional functionality on this branch
                (else
                  _))
              (else
                _))
            (if (Variables|Default|GetLostTarget)
              (Class|BGAAthenaSentryParent|LostTarget (Variables|Default|GetSelfSentryRef))
              (Development|PrintString "Sentry Lost Target")
              (else
                (Variables|Default|SetLostTarget true))))
          (:Is%20Not%20Valid
            ;;[1 1 1] If we do not have a valid target pawn then grab the first one out of the targeting array and reset the aggro timer.
            (Utilities|Time|ClearandInvalidateTimerbyHandle (Variables|/Game/Athena/Items/EnvironmentalItems/Sentry/BGA_Athena_Sentry_Parent.BGA_Athena_Sentry_Parent_C|GetTimer_BeginAggro _selfsentryref))
            (bind _array_element (Utilities|Array|ForEachLoopwithBreak _returnvalue_2)
              (:LoopBody
                (Utilities|IsValid _array_element
                  (:Is%20Valid
                    (bind _asgameplay_tag_asset_interface (Utilities|Casting|CastToGameplayTagAssetInterface _array_element))
                    (if (not (or (GameplayTags|HasAnyMatchingGameplayTags _asgameplay_tag_asset_interface (Variables|Default|GetSentryTargetingIgnoreTags)) (GameplayTags|HasAnyMatchingGameplayTags _asgameplay_tag_asset_interface (Variables|Default|GetTC_DefaultIgnoreTags))))
                      (bind (_allowtracking_2 _actortotrack) (CallFunction|CheckAttackRestrictions _array_element))
                      (if _allowtracking_2
                        (Class|BuildingGameplayActorSentry|SetActortoTrack (Variables|Default|GetSelfSentryRef) _actortotrack)))))))))))))

(fn CanTarget
  (params (Target :type Fort%20Pawn%20Object%20Reference))
  (return (not (or (Variables|/Script/FortniteGame.FortPawn|GetbIsDBNO Target) (FortPawn|IsDead Target)))))

(fn CheckAttackRestrictions
  (params (Actor :type Actor%20Object%20Reference))
  ;[0.6947 1 0.2898 0.5] Setup to allow additional checks when overridden
  (bind _selfsentryref (Variables|Default|GetSelfSentryRef)
    (:then
      (bind _actor (Variables|GetActor)
        (:then
          (return (Class|BGAAthenaSentryParent|CanAttackTargetActor _selfsentryref _actor) _actor))
        (:else)))
    (:else)))

(fn FilterTargetsInsideCone
  (params (Radius :type Float%20%28double-precision%29) (Distance :type Float%20%28double-precision%29) (ShouldCheckLOS :type Boolean) (ExcludeFromObstruction :type Gameplay%20Tag%20Container%20Structure))
  ;[0.6947 1 0.2898 0.5] Setup to allow additional/different checks when overridden
  (bind _selfsentryref (Variables|Default|GetSelfSentryRef)
    (:then
      (bind _returnvalue (Ability|TargetFilter|FilterTargetsinCone _selfsentryref (Variables|/Script/FortniteGame.BuildingGameplayActorSentry|GetTargetActorsInRange _selfsentryref) (Transformation|GetWorldLocation (Variables|/Script/FortniteGame.BuildingGameplayActorSentry|GetCameraStaticMeshComp _selfsentryref)) (+ (Transformation|GetWorldLocation (Variables|/Script/FortniteGame.BuildingGameplayActorSentry|GetCameraStaticMeshComp _selfsentryref)) (* (Math|Vector|GetForwardVector (Variables|/Script/FortniteGame.BuildingGameplayActorSentry|GetTargetRotation _selfsentryref)) (Variables|GetDistance))) (Variables|GetRadius) (Variables|GetExcludeFromObstruction) (Variables|GetShouldCheckLOS)))
      (return _returnvalue))
    (:else
      (return 0))))

(fn HandleFoundTrackedActor)
