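# Additive unlit particle material for the sentry scanner ray's endpoint, drawing two flickering rings on sprites.
# Dynamic parameters Time and Time2 size two centered sphere masks, whose values sample a gradient texture to form the rings.
# The summed rings are multiplied by a clamped sine flicker and particle color, then passed through an emissive-adjustment function.
# Particle alpha is shaped by a circular mask and a depth fade, which mobile skips.
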Material /Game/Athena/Items/EnvironmentalItems/Sentry/VFX/Materials/M_ScannerRay_Lines_Endpoint
graph {
 properties {
  bUsedWithNiagaraSprites=true
  bUsedWithParticleSprites=true
  blendMode=BLEND_Additive
  pixelDepthOffsetMode=PDOM_Legacy
  shadingModel=MSM_Unlit
  usedShadingModels=Unlit
 }
 nodes [
  MaterialFunctionCall() {materialFunction=&DebugTimeSine}
  n1 = ParticleColor()
  n2 = Multiply(A=ScalarParameter() {defaultValue=25, parameterName=FlickerSpeed}, B=Time())
  n3 = Saturate(None=Multiply(A=Sine(None=@n2)) {constB=3})
  n4 = ScalarParameter() {defaultValue=2, parameterName=RingWidth}
  n5 = Multiply(A=TextureCoordinate(), B=@n4) {constB=2}
  n6 = Multiply(A=@n4) {constB=0.5}
  n7 = DynamicParameter() {paramNames [Time,Time2,Param3,Param4]}
  n8 = Multiply(A=ScalarParameter() {defaultValue=1, parameterName=RingStartSize}, B=@n7.Time)
  n9 = Multiply(A=@n8) {constB=0.75}
  n10 = Multiply(A=@n7.Time2) {constB=0.5}
  n11 = Add(
   A=TextureSample(UVs=SphereMask(A=@n5, B=@n6, Radius=@n8, Hardness=@n9) {hardnessPercent=0}) {texture=/Game/Effects/Fort_Effects/Textures/Grad/gradient_healed}.G
   B=TextureSample(UVs=SphereMask(A=@n5, B=@n6, Radius=@n10, Hardness=@n10) {hardnessPercent=0}) {texture=&gradient_healed}.G
  )
  MaterialFunctionCall(Emissive=Multiply(A=@n1.RGB, B=Multiply(A=@n3, B=@n11))) {materialFunction=&MF_FX_EmissiveAdjustments}
  n12 = TextureCoordinate()
  n13 = Constant() {r=0.5}
  n14 = ScalarParameter() {defaultValue=15, parameterName=DepthFadeDist}
  Multiply(
   A=Multiply(A=@n1.A, B=SphereMask(A=@n12, B=@n13, Radius=@n13) {hardnessPercent=90})
   B=ShadingPathSwitch(Default=DepthFade(Opacity=@n14), Mobile=Constant() {r=1})
  )
 ]
}
